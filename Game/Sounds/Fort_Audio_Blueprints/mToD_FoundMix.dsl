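; Time-of-day ambient sound actor for a map zone, with a trigger box that fades ambience as players enter or leave.
; Day-phase callbacks swap mixes between two crossfading audio components and select random one-shots scattered around the box.
; Subzone events also set reverb.
; Destroying enough linked elements fades out and destroys the actor; the construction script can auto-gather overlapping walls, setting that threshold to half, rounded up.
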(event EventBeginPlay
  (bind _returnvalue (TimeOfDay|GetCurrentDayPhase :!class /Script/FortniteGame.FortKismetLibrary))
  (bind _returnvalue_2 (Utilities|String|EnumtoString _returnvalue))
  (bind _self self)
  ;[1 1 1] Set Up Day Phase Callbacks
  (Utilities|FlowControl|Sequence
    (:then_1
      (TimeOfDay|RegisterDayPhaseHitCallback _self "Morning_Start" "Morning" true true)
      (TimeOfDay|RegisterDayPhaseHitCallback _self "Day_Start" "Day" true)
      (TimeOfDay|RegisterDayPhaseHitCallback _self "Evening_Start" "Evening" true)
      (TimeOfDay|RegisterDayPhaseHitCallback _self "Night_Start" "Night" true)
      (TimeOfDay|RegisterDayTimeHitCallback _self "Noon" 12 true true)
      (TimeOfDay|RegisterDayTimeHitCallback _self "Midnight" 23.9 true)
      (Audio|Components|Audio|SetVolumeMultiplier (Variables|Fort_GenericTile_ToD_Ambient_BP|GetAmbientMix_A) (Variables|Ambient_Mix_Sounds|GetAmbientMix_Master_Vol) :!class /Script/Engine.AudioComponent)
      (Audio|Components|Audio|SetVolumeMultiplier (Variables|Fort_GenericTile_ToD_Ambient_BP|GetAmbientMix_B) (Variables|Ambient_Mix_Sounds|GetAmbientMix_Master_Vol) :!class /Script/Engine.AudioComponent)
      ;[1 1 1] PIE Initial ambient starter
      (Development|PrintString _returnvalue_2 (Variables|Default|GetShowAudioDebug) (Variables|Default|GetShowAudioDebug))
      (if (Utilities|String|EqualExactly(String) _returnvalue_2 "Invalid")
        (Utilities|FlowControl|Delay 1))
      (if (Utilities|String|EqualExactly(String) _returnvalue_2 "Morning")
        (CallFunction|MorningStart)
        (elif (Utilities|String|EqualExactly(String) _returnvalue_2 "Day")
          (CallFunction|DayStart)
          (elif (Utilities|String|EqualExactly(String) _returnvalue_2 "Evening")
            (CallFunction|EveningStart)
            (else
              (CallFunction|NightStart)))))
      (CallFunction|PlayRandomSounds)
      (CallFunction|SetupRoofs))))

(event Custom|Morning_Start
  ;[1 1 1] Day Phase Ambient Mix and Random Sounds Setup
  (if (not (Variables|Default|GetBlockNextPhase))
    (Audio|Components|Audio|SetSound (Variables|Fort_GenericTile_ToD_Ambient_BP|GetAmbientMix_A) (Variables|Ambient_Mix_Sounds|GetMorning_Mix_Sound)) ; Set Ambient_Mix_A to Morning
    (Variables|Default|SetAmbientMix_A_isActive true)
    (Variables|Random_Sounds|SetRandomSoundCurrent (Variables|Random_Sounds|GetRandom_Morning)) ; Set Random Sound To Current Day Phase Value
    (CallFunction|UpdateAttenuation)
    (CallFunction|PlayCrossfadeAB)))

(event Custom|Day_Start
  ;[1 1 1] Day Phase Ambient Mix and Random Sounds Setup
  (if (not (Variables|Default|GetBlockNextPhase))
    (Audio|Components|Audio|SetSound (Variables|Fort_GenericTile_ToD_Ambient_BP|GetAmbientMix_B) (Variables|Ambient_Mix_Sounds|GetDay_Mix_Sound)) ; Set Ambient_Mix_B to Day
    (Variables|Default|SetAmbientMix_A_isActive false)
    (Variables|Random_Sounds|SetRandomSoundCurrent (Variables|Random_Sounds|GetRandom_Day)) ; Set Random Sound To Current Day Phase Value
    (CallFunction|UpdateAttenuation)
    (CallFunction|PlayCrossfadeBA)))

(event OnComponentBeginOverlap(TriggerRange)
  (params (OverlappedComponent :type Primitive%20Component%20Object%20Reference) (OtherActor :type Actor%20Object%20Reference) (OtherComp :type Primitive%20Component%20Object%20Reference) (OtherBodyIndex :type Integer) (bFromSweep :type Boolean) (SweepResult :type Hit%20Result%20Structure%20%28by%20ref%29))
  ;[1 1 1] Check if subzone active
  (Utilities|Casting|CastToFortPlayerPawn OtherActor)
  (CallFunction|EnteringSubzone)
  (Variables|Default|SetIsAmbientActive true)
  (CallFunction|PlayRandomSounds)
  (CallFunction|FadeinAudio)
  (Development|PrintString "YOU MOVED TO THE............................................................................................................... FOUNDMIX AREA" (Variables|Default|GetShowAudioDebug) (Variables|Default|GetShowAudioDebug) "(R=0.500000,G=0.047730,B=0.019064,A=1.000000)"))

(event Custom|Evening_Start
  ;[1 1 1] Day Phase Ambient Mix and Random Sounds Setup
  (if (not (Variables|Default|GetBlockNextPhase))
    (Audio|Components|Audio|SetSound (Variables|Fort_GenericTile_ToD_Ambient_BP|GetAmbientMix_A) (Variables|Ambient_Mix_Sounds|GetEvening_Mix_Sound)) ; Set Ambient_Mix_A to Evening
    (Variables|Default|SetAmbientMix_A_isActive true)
    (Variables|Random_Sounds|SetRandomSoundCurrent (Variables|Random_Sounds|GetRandom_Evening)) ; Set Random Sound To Current Day Phase Value
    (CallFunction|UpdateAttenuation)
    (CallFunction|PlayCrossfadeAB)))

(event OnComponentEndOverlap(TriggerRange)
  (params (OverlappedComponent :type Primitive%20Component%20Object%20Reference) (OtherActor :type Actor%20Object%20Reference) (OtherComp :type Primitive%20Component%20Object%20Reference) (OtherBodyIndex :type Integer))
  ;[1 1 1] Check if subzone active
  (Utilities|Casting|CastToFortPlayerPawn OtherActor)
  (CallFunction|LeavingSubzone)
  (Variables|Default|SetIsAmbientActive false)
  (CallFunction|StopRandomSounds)
  (CallFunction|FadeoutAudio)
  (Development|PrintString "YOU MOVED TO THE............................................................................................................... FOUNDMIX FALLOF" (Variables|Default|GetShowAudioDebug) (Variables|Default|GetShowAudioDebug) "(R=0.823529,G=0.000000,B=0.439215,A=1.000000)"))

(event Custom|Night_Start
  ;[1 1 1] Day Phase Ambient Mix and Random Sounds Setup
  (if (not (Variables|Default|GetBlockNextPhase))
    (Audio|Components|Audio|SetSound (Variables|Fort_GenericTile_ToD_Ambient_BP|GetAmbientMix_B) (Variables|Ambient_Mix_Sounds|GetNight_Mix_Sound)) ; Set Ambient_Mix_B to Night
    (Variables|Default|SetAmbientMix_A_isActive false)
    (Variables|Random_Sounds|SetRandomSoundCurrent (Variables|Random_Sounds|GetRandom_Night)) ; Set Random Sound To Current Day Phase Value
    (CallFunction|UpdateAttenuation)
    (CallFunction|PlayCrossfadeBA)))

(event Custom|PlayRandomSounds)

(event Custom|Entering_subzone
  (bind _tilereverbtag (Variables|Reverb_Settings|GetTileReverbTag))
  ;[1 1 1] Random Sounds Player
  (Utilities|FlowControl|Sequence
    (:then_0
      _)
    (:then_1
      (Utilities|FlowControl|Delay 0.5)))
  (Utilities|FlowControl|Gate true
    (:Exit
      (Utilities|FlowControl|Sequence
        (:then_0
          (if (FortGlobals|IsinAthena)
            (Audio|ActivateReverbEffect (Variables|Reverb_Settings|GetTile_Reverb_Preset) _tilereverbtag)
            (else
              (Audio|ActivateReverbEffect "/Game/Sounds/Reverb_Presets/Fort_FNBR_Outdoor_Reverb.Fort_FNBR_Outdoor_Reverb" _tilereverbtag 1.0 0.9))))
        (:then_1
          (Utilities|FlowControl|Gate true
            (:Exit
              (Utilities|FlowControl|Delay (Math|Random|RandomFloatinRange (Variables|Random_Sounds|GetRandomSound_MinDelay) (Variables|Random_Sounds|GetRandomSound_MaxDelay))))))
        (:then_2
          _))
      (Audio|PlaySoundatLocation (Variables|Random_Sounds|GetRandomSoundCurrent) (Math|Vector|MakeVector (+ (Math|Random|RandomFloatinRange -3000 3000) (.x (Transformation|GetWorldLocation (Variables|Fort_GenericTile_ToD_Ambient_BP|GetTriggerRange)))) (+ (Math|Random|RandomFloatinRange -3000 3000) (.y (Transformation|GetWorldLocation (Variables|Fort_GenericTile_ToD_Ambient_BP|GetTriggerRange)))) (.z (Transformation|GetWorldLocation (Variables|Fort_GenericTile_ToD_Ambient_BP|GetTriggerRange))))) ; Play Random Location Sound
    )))

(event Custom|Leaving_subzone
  (bind _tilereverbtag (Variables|Reverb_Settings|GetTileReverbTag))
  ;[1 1 1] Random Sounds Player
  (Utilities|FlowControl|Sequence
    (:then_0
      (Utilities|FlowControl|Gate true
        (:Exit
          (Utilities|FlowControl|Sequence
            (:then_0
              (if (FortGlobals|IsinAthena)
                (Audio|ActivateReverbEffect (Variables|Reverb_Settings|GetTile_Reverb_Preset) _tilereverbtag)
                (else
                  (Audio|ActivateReverbEffect "/Game/Sounds/Reverb_Presets/Fort_FNBR_Outdoor_Reverb.Fort_FNBR_Outdoor_Reverb" _tilereverbtag 1.0 0.9))))
            (:then_1
              (Utilities|FlowControl|Gate true
                (:Exit
                  (Utilities|FlowControl|Delay (Math|Random|RandomFloatinRange (Variables|Random_Sounds|GetRandomSound_MinDelay) (Variables|Random_Sounds|GetRandomSound_MaxDelay))))))
            (:then_2
              _))
          (Audio|PlaySoundatLocation (Variables|Random_Sounds|GetRandomSoundCurrent) (Math|Vector|MakeVector (+ (Math|Random|RandomFloatinRange -3000 3000) (.x (Transformation|GetWorldLocation (Variables|Fort_GenericTile_ToD_Ambient_BP|GetTriggerRange)))) (+ (Math|Random|RandomFloatinRange -3000 3000) (.y (Transformation|GetWorldLocation (Variables|Fort_GenericTile_ToD_Ambient_BP|GetTriggerRange)))) (.z (Transformation|GetWorldLocation (Variables|Fort_GenericTile_ToD_Ambient_BP|GetTriggerRange))))) ; Play Random Location Sound
        )))
    (:then_1
      _)
    (:then_2
      (if (FortGlobals|IsinAthena)
        (Audio|DeactivateReverbEffect _tilereverbtag)))))

(event Custom|StopRandomSounds
  ;[1 1 1] Random Sounds Player
  (Utilities|FlowControl|Gate true
    (:Exit
      (Utilities|FlowControl|Delay (Math|Random|RandomFloatinRange (Variables|Random_Sounds|GetRandomSound_MinDelay) (Variables|Random_Sounds|GetRandomSound_MaxDelay)))
      (Audio|PlaySoundatLocation (Variables|Random_Sounds|GetRandomSoundCurrent) (Math|Vector|MakeVector (+ (Math|Random|RandomFloatinRange -3000 3000) (.x (Transformation|GetWorldLocation (Variables|Fort_GenericTile_ToD_Ambient_BP|GetTriggerRange)))) (+ (Math|Random|RandomFloatinRange -3000 3000) (.y (Transformation|GetWorldLocation (Variables|Fort_GenericTile_ToD_Ambient_BP|GetTriggerRange)))) (.z (Transformation|GetWorldLocation (Variables|Fort_GenericTile_ToD_Ambient_BP|GetTriggerRange))))) ; Play Random Location Sound
    )))

(event Custom|FadeoutAudio
  ;[1 1 1] Fading audio
  (if (Variables|Default|GetClearAmbient)
    _
    (elif (Variables|Default|GetAmbientMix_A_isActive)
      (Audio|Components|Audio|FadeOut (Variables|Fort_GenericTile_ToD_Ambient_BP|GetAmbientMix_A) 3.0 0.01 "Linear" :!class /Script/Engine.AudioComponent)
      (CallFunction|StopRandomSounds)
      (else
        (Audio|Components|Audio|FadeOut (Variables|Fort_GenericTile_ToD_Ambient_BP|GetAmbientMix_B) 3.0 0.01 "Linear" :!class /Script/Engine.AudioComponent)))))

(event Custom|PlayCrossfadeAB
  ;[1 1 1] Ambient Mix Player (Crossfade)
  (if (Variables|Default|GetIsAmbientActive)
    (Audio|Components|Audio|FadeIn (Variables|Fort_GenericTile_ToD_Ambient_BP|GetAmbientMix_A) 5 1.0 0.0 "Linear" :!class /Script/Engine.AudioComponent) ; Fade In The Morning Mix
    (Audio|Components|Audio|FadeOut (Variables|Fort_GenericTile_ToD_Ambient_BP|GetAmbientMix_B) 5 0.001 "Linear" :!class /Script/Engine.AudioComponent) ; Fade Out The Night Mix
  ))

(event Custom|PlayCrossfadeBA
  ;[1 1 1] Ambient Mix Player (Crossfade)
  (if (Variables|Default|GetIsAmbientActive)
    (Audio|Components|Audio|FadeIn (Variables|Fort_GenericTile_ToD_Ambient_BP|GetAmbientMix_B) 5 1.0 0.0 "Linear" :!class /Script/Engine.AudioComponent) ; Fade In The Morning Mix
    (Audio|Components|Audio|FadeOut (Variables|Fort_GenericTile_ToD_Ambient_BP|GetAmbientMix_A) 5 0.001 "Linear" :!class /Script/Engine.AudioComponent) ; Fade Out The Night Mix
  ))

(event Custom|FadeinAudio
  ;[1 1 1] Fading audio
  (if (Variables|Default|GetClearAmbient)
    _
    (elif (Variables|Default|GetAmbientMix_A_isActive)
      (Audio|Components|Audio|FadeIn (Variables|Fort_GenericTile_ToD_Ambient_BP|GetAmbientMix_A) 5.0 1.0 0.0 "Linear" :!class /Script/Engine.AudioComponent)
      (CallFunction|PlayRandomSounds)
      (else
        (Audio|Components|Audio|FadeIn (Variables|Fort_GenericTile_ToD_Ambient_BP|GetAmbientMix_B) 5.0 1.0 0.0 "Linear" :!class /Script/Engine.AudioComponent)))))

(event Custom|SetupRoofs
  ;[1 1 1] Roof detector
  (for _array_element (Variables|AmbientConnections|GetElements%20List)
    (Utilities|IsValid _array_element
      (:Is%20Valid
        (Game|BindEventtoOnDestroyed _array_element (delegate Custom|ActorDestroyedSignature__DelegateSignature_Event))))))

(event Custom|ManualFadeOut
  (bind _on_destroy_fade_out_time (Variables|AmbientConnections|GetOn%20Destroy%20Fade%20Out%20Time))
  ;[1 1 1] Final fade out on destroy
  (Variables|Default|SetClearAmbient true)
  (CallFunction|StopRandomSounds)
  (if (Variables|Default|GetAmbientMix_A_isActive)
    (Audio|Components|Audio|FadeOut (Variables|Fort_GenericTile_ToD_Ambient_BP|GetAmbientMix_A) _on_destroy_fade_out_time 0.0 "Linear" :!class /Script/Engine.AudioComponent)
    (else
      (Audio|Components|Audio|FadeOut (Variables|Fort_GenericTile_ToD_Ambient_BP|GetAmbientMix_B) _on_destroy_fade_out_time 0.0 "Linear" :!class /Script/Engine.AudioComponent)))
  (Utilities|FlowControl|Delay _on_destroy_fade_out_time)
  (Actor|DestroyActor))

(event Custom|ActorDestroyedSignature__DelegateSignature_Event
  (params (DestroyedActor :type Actor%20Object%20Reference))
  (bind _elementsdestroyed (Variables|AmbientConnections|GetElementsDestroyed))
  ;[1 1 1] Roof detector
  (Development|PrintString "Actor destroyed" (Variables|Default|GetShowAudioDebug) (Variables|Default|GetShowAudioDebug))
  (Variables|AmbientConnections|SetElementsDestroyed (+ _elementsdestroyed 1))
  (if (>= _elementsdestroyed (Math|Integer|Min(Integer) (Utilities|Array|Length (Variables|AmbientConnections|GetElements%20List)) (Variables|AmbientConnections|GetNo%20of%20Elements%20to%20Deactivate)))
    (Variables|Default|SetBlockNextPhase true)
    (CallFunction|ManualFadeOut)))

(event Custom|Update%20attenuation)

(fn ConstructionScript
  (bind _elements_list (Variables|AmbientConnections|GetElements%20List))
  ;[1 1 1] Setup Range Visualisation
  (Components|Box|SetBoxExtent (Variables|Fort_GenericTile_ToD_Ambient_BP|GetTriggerRange) (Variables|AttenuationSettings|GetAttenuation%20Extents) true)
  (Variables|Default|SetShowAudioDebug false)
  (if (Variables|AmbientConnections|GetDelete%20Repeated)
    ;[1 1 1] Deleting Repeated Walls
    (Variables|Automation|SetTempElementList _elements_list)
    (Utilities|Array|Clear _elements_list)
    (for _array_element (Variables|Automation|GetTempElementList)
      (Utilities|Array|AddUnique (Variables|AmbientConnections|GetElements%20List) _array_element))
    (Variables|AmbientConnections|SetDelete%20Repeated false)
    (elif (Variables|AmbientConnections|GetAddAutomatically)
      ;[1 1 1] Adding Walls Automatically
      (bind _outactors (Collision|ComponentOverlapActors (Variables|Fort_GenericTile_ToD_Ambient_BP|GetTriggerRange self) (Transformation|GetActorTransform self) (Utilities|Array|MakeArray "ObjectTypeQuery1") "/Script/FortniteGame.BuildingWall"))
      (Utilities|Array|Clear (Variables|AmbientConnections|GetElements%20List self))
      (for _array_element_2 _outactors
        (Utilities|Array|AddUnique (Variables|AmbientConnections|GetElements%20List self) _array_element_2))
      (Variables|AmbientConnections|SetAddAutomatically false)
      (Variables|AmbientConnections|SetNo%20of%20Elements%20to%20Deactivate (+ (/ (Utilities|Array|Length (Variables|AmbientConnections|GetElements%20List)) 2) (Math|Integer|%25%28Integer%29 (Utilities|Array|Length (Variables|AmbientConnections|GetElements%20List)) 2))))))
